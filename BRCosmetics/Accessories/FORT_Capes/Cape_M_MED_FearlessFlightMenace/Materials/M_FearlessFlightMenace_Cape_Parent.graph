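# Parent material for the FearlessFlightMenace cape cosmetic, using material attributes with Default Lit and Subsurface Profile shading on skinned, cloth-simulated meshes.
# The shared character PreFX function feeds a FearlessFlightMenace HoloDots function, then passes through character PostFX and cinematic PostFX.
# A SkinFX mask drives HoloDots through a lerp whose inputs are both the mask RGB, so UseSkinFXMaskForOpacity has no effect.
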Material /BRCosmetics/Accessories/FORT_Capes/Cape_M_MED_FearlessFlightMenace/Materials/M_FearlessFlightMenace_Cape_Parent
graph {
 properties {
  bAutomaticallySetUsageInEditor=false
  bUseMaterialAttributes=true
  bUsedWithClothing=true
  bUsedWithMeshDeformer=true
  bUsedWithMorphTargets=true
  bUsedWithSkeletalMesh=true
  pixelDepthOffsetMode=PDOM_Legacy
  shadingModel=MSM_FromMaterialExpression
  subsurfaceProfile=&SS_HeroSkin_02
  usedShadingModels=Default Lit | Subsurface Profile
 }
 nodes [
  n1 = MaterialFunctionCall() {materialFunction=&BaseFN_CharacterFunction_PreFX}
  n2 = TextureSampleParameter2D() {
   group=Skin FX
   parameterName=SkinFX_Mask
   samplerType=SAMPLERTYPE_Masks
   sortPriority=1
   texture=&T_BlackMask
  }
  n3 = ScalarParameter() {
   desc=Set to 1 if the Red Channel in the FX Mask should be counted for Hologram masking in the FE
   group=HologramMasking
   parameterName=UseSkinFXMaskForOpacity
  }
  n4 = MaterialFunctionCall(
   AO=@n1.AO
   Distance Blend=@n1.AO
   Fresnel=@n1.AO
   Gmap=@n1.AO
   Pre Skinned Local Position=@n1.AO
   Material=MaterialFunctionCall!(MA=@n1.Result, FX=LinearInterpolate(A=@n2.RGB, B=@n2.RGB, Alpha=@n3)) {materialFunction=&/MFC_FearlessFlightMenace_HoloDots}.Result
  ) {materialFunction=&BaseFN_CharacterFunction_PostFX}
  MaterialFunctionCall(In=@n4.Result) {materialFunction=&CINE_MF_CharPostFX}
 ]
}
